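# Translucent unlit material for Niagara sprites, by name and folder a half-star effect for the demon kaiju's sword.
# Builds a procedural star mask from switchable UV sets, using square roots of centered absolute coordinates and a ratio between axes, faded by particle alpha.
# A three-way lerp feeds a named reroute, and a reciprocal-style contrast branch reshapes a reroute value.
# No links to material outputs appear, and the motion-vector and ramp functions are unwired.
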Material /GrillMoon/Assets/FX/Kaiju/Fight/Demon/Sword/M_HalfStar
graph {
 properties {
  bAutomaticallySetUsageInEditor=false
  bUsedWithNiagaraSprites=true
  blendMode=BLEND_Translucent
  shadingModel=MSM_Unlit
  usedShadingModels=Unlit
 }
 nodes [
  MaterialFunctionCall!(MotionVector, SubUVtexture, Strength, "DirectionSwitchRG(1.-1)") {materialFunction=/Game/Effects/Fort_Effects/FrontEnd/S33/GateContent/MF_MOON_SubUVMotionVectors_V2}
  MaterialFunctionCall!(InRamp, CoreIntensity, DetailBalence, MaxOutPut) {materialFunction=/Hermes/Props/ClayTile/ClayTile_DragonGate_B/VFX/MF_MOON_InvertDistanceRamp}
  n1 = OneMinus(None=NamedRerouteUsage())
  n2 = ScalarParameter() {defaultValue=1}
  MaterialFunctionCall(In=Subtract(A=Divide(A=@n2, B=@n1), B=@n2), Contrast=ScalarParameter() {defaultValue=11, parameterName=Param_1}) {materialFunction=&CheapContrast}
  n3 = Multiply(A=LinearInterpolate(A=TextureCoordinate(), B=TextureCoordinate() {coordinateIndex=1}, Alpha=ScalarParameter() {parameterName=CubeMap_Switch})) {constB=2}
  n4 = SquareRoot(None=Abs(None=Subtract(A=@n3)))
  n5 = MaterialFunctionCall(Float2=@n4) {materialFunction=&BreakOutFloat2Components}
  n6 = MaterialFunctionCall(Float2=OneMinus(None=@n4)) {materialFunction=&BreakOutFloat2Components}
  Multiply(A=ParticleColor().A, B=Subtract(B=Divide(A=@n5.G, B=@n6.R)))
  NamedRerouteDeclaration(None=MaterialFunctionCall(A=@n6.R, B=Multiply(A=@n6.R, B=@n6.R), C=@n6.R, Alpha=ScalarParameter() {defaultValue=0.5, parameterName=CubeMap_Select}) {materialFunction=&Lerp_3Color}.Result) {name=CubeMap_2, nodeColor=<rgba 1 1 1 1>}
  TextureCoordinate()
 ]
}
