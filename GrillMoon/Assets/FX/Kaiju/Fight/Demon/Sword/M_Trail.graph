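# Unlit masked material for Niagara ribbons, by its path a sword trail effect for the Demon kaiju fight.
# Inverted UVs are split into U and V, which feed a product and a separate sum-then-subtract term scaled by particle alpha.
# Particle alpha also feeds a constant multiply, and the opacity mask clip value is zero.
# The graph records no links from these branches to material outputs.
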Material /GrillMoon/Assets/FX/Kaiju/Fight/Demon/Sword/M_Trail
graph {
 properties {
  bAutomaticallySetUsageInEditor=false
  bUsedWithNiagaraRibbons=true
  blendMode=BLEND_Masked
  opacityMaskClipValue=0
  shadingModel=MSM_Unlit
  usedShadingModels=Unlit
 }
 nodes [
  n1 = ParticleColor()
  Multiply(A=Constant() {r=1}, B=@n1.A)
  n2 = OneMinus(None=TextureCoordinate())
  n3 = ComponentMask(None=@n2) {r=true}
  n4 = ComponentMask(None=@n2) {g=true}
  Multiply(A=@n3, B=@n4)
  Multiply(A=Subtract(A=Add(A=@n3, B=@n4)), B=@n1.A) {constB=999}
 ]
}
