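# Additive unlit Niagara sprite material, by its name and path a void background for the demon kaiju's sword fight FX.
# A soft UV sphere mask and named reroutes organise the graph, with particle color tinting a squared inverted mask.
# Alpha comes from a strength-scaled mask through depth fade, multiplied by particle alpha.
# 2D-offset refraction builds a radial vector from centered, Y-flipped UVs, scaled by mask, Ref parameter and particle alpha.
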Material /GrillMoon/Assets/FX/Kaiju/Fight/Demon/Sword/M_VoidBG
graph {
 properties {
  bAllowNegativeEmissiveColor=true
  bAutomaticallySetUsageInEditor=false
  bUsedWithNiagaraSprites=true
  blendMode=BLEND_Additive
  refractionMethod=RM_2DOffset
  shadingModel=MSM_Unlit
  usedShadingModels=Unlit
 }
 nodes [
  n1 = NamedRerouteUsage()
  n2 = OneMinus(None=@n1)
  n3 = Multiply(A=@n2, B=@n2)
  Multiply(A=ParticleColor().RGB, B=@n3)
  n4 = DepthFade(
   Opacity=Multiply(A=@n1, B=ScalarParameter() {defaultValue=21.2431, parameterName=AlphaStrength}) {constB=99}
   FadeDistance=ScalarParameter() {defaultValue=10000, parameterName=DepthFade}
  )
  Multiply(A=ParticleColor().A, B=Saturate(None=NamedRerouteDeclaration(None=@n4) {name=Alpha, nodeColor=<rgba 1 1 1 1>}))
  n5 = NamedRerouteUsage()
  n6 = Subtract(A=TextureCoordinate()) {constB=0.5}
  n7 = Constant2Vector() {g=-1, r=1}
  n8 = Multiply(A=@n3, B=Normalize(VectorInput=Multiply(A=@n6, B=@n7)))
  Multiply(A=Multiply(A=Multiply(A=@n5, B=@n8), B=ScalarParameter() {defaultValue=0.0366, parameterName=Ref}), B=Saturate(None=ParticleColor().A))
  n9 = Constant() {r=0.5}
  NamedRerouteDeclaration(None=SphereMask(A=TextureCoordinate(), B=@n9, Radius=@n9) {hardnessPercent=0}) {name=SimpleSphereMask, nodeColor=<rgba 1 1 1 1>}
  NamedRerouteUsage()
 ]
}
