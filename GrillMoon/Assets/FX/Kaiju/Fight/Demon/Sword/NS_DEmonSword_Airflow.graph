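# Niagara burst effect for the demon kaiju's sword airflow, sized and tinted by system-level FX scale and character tint values.
# Minimal bursts mesh strands, with count switched by a platform quality set and rotation fanned by partition index.
# SoftAirflow adds growing, spinning meshes and upward-facing sprites, feeding a time-of-day scalar to the material.
# Both emitters fade alpha over lifetime and animate their materials through dynamic parameters.
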NiagaraSystem /GrillMoon/Assets/FX/Kaiju/Fight/Demon/Sword/NS_DEmonSword_Airflow
dependencies {
 usedDataInterfaces [&NiagaraDataInterfaceCurve, &NiagaraDataInterfacePlatformSet]
 usedDynamicInputs [&LerpLinearColors]
 usedModules [
  &self:SystemSpawnScript.SystemScriptSource.SystemScriptGraph.NiagaraNodeAssignment_0.SetVariables_#1
  &self:SystemSpawnScript.SystemScriptSource.SystemScriptGraph.NiagaraNodeAssignment_1.SetVariables_#2
  &self:SystemSpawnScript.SystemScriptSource.SystemScriptGraph.NiagaraNodeAssignment_2.SetVariables_#3
  &SystemState
 ]
}
emitters {
 Minimal {
  emitterUpdateScript [
   &EmitterState()
   /SetVariables(Emitter.Num Strands:NiagaraInt32=&MakeCustomIntFromBool(
     Bool=&IsPlatformSetActive(New Platform Set=DataInterface() {propertyValues {Platforms {qualityLevelMask=-29}}})
     True INT=3
     False INT=7
    )) {isSetParameters=true}
   &SpawnBurst_Instantaneous(Spawn Count=$Emitter.Num Strands, Use Spawn Probability=false, UseLoopCountLimit=false)
  ]
  particleSpawnScript [
   &PartitionParticles(Number Of Partitions=1, Particles Per Partition:NiagaraInt32=$Emitter.Num Strands, Partition Mode:ENiagara_PartitionMode=::NewEnumerator1)
   &InitializeParticle(
    Lifetime Max=1
    Lifetime Min=1.35
    Lifetime Mode=::NewEnumerator1
    Mesh Scale Max=<1 1 0.5>
    Mesh Scale Min=<1 1 0.2>
    Mesh Scale Mode=::NewEnumerator2
    Position Offset=<20 -120 70>
    UsePositionOffset=true
    Write Material Random=true
   )
   &InitialMeshOrientation(Mesh Orientation Mode=::NewEnumerator7, Rotation=<0 0 0.25>, Use Rotation Vector=true)
   &ShapeLocation(
    Handle Radius=1
    Large Radius=20
    MaskSpawnGroup=false
    Override Seed=false
    Rotation Axis=<0 1 0>
    Rotation Mode=::NewEnumerator1
    Shape Primitive=::NewEnumerator3
    UseOwnerScale=true
   )
   &InitialMeshOrientation(
    Mesh Orientation Mode=::NewEnumerator7
    Rotation=&MakeVector(
     X=0
     Y=0
     Z=&Add_Float(
      A=&FloatFromCurve(
       FloatCurve=DataInterface() {propertyValues {Curve {keys [<at 0 1>,<at 1 0>]}}}
       CurveIndex=$Particles.PartitionParticles.Normalized Index In Partition
      )
      B=&Multiply_Float(A=0.15, B=&RandomRangeFloat(Minimum=-1, Maximum=1))
     )
    )
    Use Rotation Vector=true
   ) {module=InitialMeshOrientation001}
   &InitialMeshOrientation(
    Mesh Orientation Mode=::NewEnumerator7
    Rotation=&RandomRangeVector(Minimum=<0 0 0>, Maximum=<0 0 0>)
    Use Rotation Vector=true
   ) {module=InitialMeshOrientation002}
   &CameraOffset(Camera Offset Amount=50)
  ]
  particleUpdateScript [
   &ParticleState()
   &ScaleMeshSize(Scale Factor=&Multiply_Vector(
     A=&Multiply_VectorByFloat(Vector=&RandomRangeVector(Minimum=<2 5 2>, Maximum=<2.5 4 3>), Float=0.7)
     B=&VectorFromFloat(Value=$System.FX scale)
    )) {module=ScaleMeshSize001}
   &ScaleMeshSize(Scale Factor=&Multiply_VectorByFloat(
     Vector=&MakeVector(
      X=&FloatFromCurve(FloatCurve=DataInterface() {propertyValues {Curve {keys [<at 0 1>,<at 1 1.5>]}}})
      Y=&FloatFromCurve(FloatCurve=DataInterface() {propertyValues {Curve {keys [<at 0 1>,<at 1 1>]}}})
      Z=&FloatFromCurve(FloatCurve=DataInterface() {propertyValues {Curve {keys [<at 0 1>,<at 1 2>]}}})
     )
     Float=&RandomRangeFloat(Minimum=0.3, Maximum=0.3)
    ))
   &ScaleColor(
    Scale Alpha=&FloatFromCurve(
     FloatCurve=DataInterface() {propertyValues {Curve {keys [
         {interpMode=RCIM_Cubic, tangentMode=RCTM_User, time=0, value=0.5}
         {arriveTangent=-1.7032, interpMode=RCIM_Cubic, leaveTangent=-1.7032, tangentMode=RCTM_User, time=0.2, value=0.6}
         {arriveTangent=-0.1551, tangentMode=RCTM_User, time=1, value=0}
        ]}}}
     Scale Curve=&RandomRangeFloat(Minimum=0.5, Maximum=0.7)
    )
    Scale RGB=&MakeVectorFromLinearColorRGB(Color=$System.Char Tint)
    ScaleA=true
    ScaleRGB=true
   )
   &DynamicMaterialParameters(
    Index 0 Param 1=&FloatFromCurve(
     FloatCurve=DataInterface() {propertyValues {Curve {keys [
         {arriveTangent=2.8589, interpMode=RCIM_Cubic, leaveTangent=2.8589, tangentMode=RCTM_User, time=0.5, value=0}
         {arriveTangent=1.962, tangentMode=RCTM_User, time=1, value=0.7}
        ]}}}
     Scale Curve=50
    )
    Index 0 Param 3=&Multiply_Float(
     A=&Multiply_Float(A=$Engine.Time, B=0.5)
     B=&Add_Float(A=-2, B=&RandomRangeFloat(Minimum=0, Maximum=1))
    )
    Index 2 Param 1=&Multiply_Float(
     A=$System.FX scale
     B=&Multiply_Float(
      A=&RandomRangeFloat(Minimum=40, Maximum=35)
      B=&FloatFromCurve(FloatCurve=DataInterface() {propertyValues {Curve {keys [{interpMode=RCIM_Cubic, tangentMode=RCTM_User, time=0, value=1}, {interpMode=RCIM_Cubic, tangentMode=RCTM_User, time=1, value=0.8}]}}})
     )
    )
    Index 2 Param 2=&FloatFromCurve(FloatCurve=DataInterface() {propertyValues {Curve {keys [{arriveTangent=2.2418, time=0, value=1}, <at 1 1 RCIM_Cubic>]}}})
    Index 2 Param 3=&FloatFromCurve(FloatCurve=DataInterface() {propertyValues {Curve {keys [<at 0 0>,<at 1 1>]}}})
    Module.Index0Param0Write=true
    Module.Index0Param1Write=false
    Module.Index0Param2Write=true
    Module.Index0Param3Write=false
    Module.Index2Param0Write=true
    Module.Index2Param1Write=true
    Module.Index2Param2Write=true
    Module.Index2Param3Write=true
    Write Parameter Index 2=true
   ) {module=DynamicMaterialParameters001}
   &ScaleColor(Scale Alpha=5, ScaleA=true, ScaleRGB=false) {module=ScaleColor001}
  ]
  rendererClasses [&NiagaraMeshRendererProperties]
  renderers [{rendererClass=&NiagaraMeshRendererProperties, rendererIndex=0}]
 }
 SoftAirflow {
  emitterUpdateScript [
   &EmitterState()
   &SpawnBurst_Instantaneous(Spawn Count=5, Use Spawn Probability=false, UseLoopCountLimit=false)
  ]
  particleSpawnScript [
   &InitializeParticle(
    Color=<rgba 0.8698 0.8698 0.8698 1>
    Lifetime Min=1.2
    Lifetime Mode=::NewEnumerator1
    Mesh Scale=&Multiply_VectorByFloat(Vector=<1 1 1>, Float=&RandomRangeFloat(Minimum=2, Maximum=3))
    Mesh Scale Mode=::NewEnumerator4
    Position Offset=<0 0 200>
    Sprite Size Mode=::NewEnumerator3
    Uniform Sprite Size=312.4647
    UsePositionOffset=false
    Write Material Random=true
   )
   &InitialMeshOrientation(
    Mesh Orientation Mode=::NewEnumerator7
    Rotation=&RandomRangeVector(Minimum=<0 0 -1>, Maximum=<0 0 1>)
    Use Rotation Vector=true
   )
   &ShapeLocation(
    Cylinder Height=3
    Cylinder Radius=0
    MaskSpawnGroup=false
    Non Uniform Scale=&VectorFromFloat(Value=$System.FX scale)
    Override Seed=false
    Shape Primitive=::NewEnumerator1
    Surface Only=false
    UseOwnerScale=true
   )
  ]
  particleUpdateScript [
   &ParticleState()
   &ScaleSpriteSize(
    Uniform Curve Scale=$System.FX scale
    Uniform Curve Sprite Scale=DataInterface() {propertyValues {Curve {keys [{tangentMode=RCTM_User, time=0, value=0.5}, {tangentMode=RCTM_User, time=1, value=1}]}}}
   )
   &ScaleMeshSize(Scale Factor=&VectorFromFloat(Value=$System.FX scale)) {module=ScaleMeshSize001}
   &DynamicMaterialParameters(
    Index 0 Param 1=&FloatFromCurve(FloatCurve=DataInterface() {propertyValues {Curve {keys [
         {interpMode=RCIM_Cubic, tangentMode=RCTM_User, time=0.5, value=0}
         {arriveTangent=7.2001, interpMode=RCIM_Cubic, leaveTangent=7.2001, tangentMode=RCTM_User, time=1, value=1}
        ]}}})
    Index 0 Param 2=&FNTimeOfDayScalar(Night Scale=0, Morning Scale=1, Evening Scale=1)
    Index 0 Param 3=1
    Module.Param0WriteEnabled=true
    Module.Param1WriteEnabled=true
    Module.Param2WriteEnabled=true
    Module.Param3WriteEnabled=true
   )
   &ScaleMeshSize(Scale Factor=&Multiply_VectorByFloat(
     Vector=<1 1 1.5>
     Float=&FloatFromCurve(
      FloatCurve=DataInterface() {propertyValues {Curve {keys [
          {arriveTangent=0.6673, interpMode=RCIM_Cubic, leaveTangent=0.6673, tangentMode=RCTM_User, time=0, value=0}
          {arriveTangent=0.1286, interpMode=RCIM_Cubic, leaveTangent=0.1286, tangentMode=RCTM_User, time=1, value=1}
         ]}}}
      Scale Curve=1.1
     )
    ))
   &UpdateMeshOrientation(Rotation Rate=&RandomRangeFloat(Minimum=-0.25, Maximum=0.35), Rotation Vector=<0 0 1>)
   &ScaleColor(
    Scale Alpha=&FloatFromCurve(
     FloatCurve=DataInterface() {propertyValues {Curve {keys [
         {arriveTangent=1.8077, interpMode=RCIM_Cubic, leaveTangent=1.8077, tangentMode=RCTM_User, time=0, value=0}
         {arriveTangent=1, interpMode=RCIM_Cubic, leaveTangent=1, time=0.2, value=1.4}
         {arriveTangent=-3.8474, interpMode=RCIM_Cubic, leaveTangent=-4.7531, tangentMode=RCTM_Break, time=0.5, value=0.5}
         {arriveTangent=-0.7636, interpMode=RCIM_Cubic, leaveTangent=-0.7636, tangentMode=RCTM_User, time=1, value=0}
        ]}}}
     Scale Curve=0.45
    )
    Scale RGB=&MakeVectorFromLinearColorRGB(Color=$System.Char Tint)
    ScaleA=true
    ScaleRGB=true
   )
   &SpriteFacingAndAlignment(Sprite Facing=<0 0 1>, WriteAlignment=true, WriteFacing=true)
   /SetVariables(Particles.Sprote color:LinearColor=&MultiplyLinearColor(
     Linear Color=$Particles.Color
     Scale Factor=&MakeFloat4FromVectorAndFloat("Vector (RGB)"=&VectorFromFloat(Value=1), "Float (Alpha)"=0.2)
    )) {isSetParameters=true}
  ]
  rendererClasses [&NiagaraMeshRendererProperties, &NiagaraSpriteRendererProperties]
  renderers [
   {rendererClass=&NiagaraMeshRendererProperties, rendererIndex=0}
   {rendererClass=&NiagaraSpriteRendererProperties, rendererIndex=1}
  ]
 }
}
summary {
 systemName=NS_DEmonSword_Airflow
 emitters [
  {emitterName=SoftAirflow, rendererClasses [&NiagaraMeshRendererProperties, &NiagaraSpriteRendererProperties]}
  {emitterName=Minimal, rendererClasses [&NiagaraMeshRendererProperties]}
 ]
 userVariables [User.CharTint {type=NiagaraFloat, defaultValue=0, description=""}]
}
