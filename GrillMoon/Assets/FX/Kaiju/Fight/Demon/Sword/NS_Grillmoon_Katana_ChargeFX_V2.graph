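# Niagara charge-up effect for the Demon kaiju's katana, built from oversized, camera-offset sprites.
# Star bursts form spinning crosses using two renderers with swapped size axes: a hot magenta glint and a wider red-purple layer.
# Magenta streaks converge on the emitter by setting position from random directions and a shrinking distance curve.
# A negative-colour Void sprite moves through depth, swelling as it fades; the scattered Star001 emitter is disabled.
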NiagaraSystem /GrillMoon/Assets/FX/Kaiju/Fight/Demon/Sword/NS_Grillmoon_Katana_ChargeFX_V2
dependencies {
 usedDataInterfaces [&NiagaraDataInterfaceCurve, &NiagaraDataInterfaceColorCurve]
 usedDynamicInputs []
 usedModules [
  &self:SystemSpawnScript.SystemScriptSource.SystemScriptGraph.NiagaraNodeAssignment_1.SetVariables_#1
  &SystemState
 ]
}
emitters {
 Star {
  emitterUpdateScript [
   &EmitterState()
   &SpawnBurst_Instantaneous(Use Spawn Probability=false, UseLoopCountLimit=false)
  ]
  particleSpawnScript [&InitializeParticle(
    Color Mode=::NewEnumerator0
    Sprite Rotation Mode=::NewEnumerator0
    Sprite Size=&Multiply_Vector2D_ByFloat(Float=1000, Vector2D=<100 1>)
    Sprite Size Mode=::NewEnumerator4
    UsePositionOffset=false
    Write Material Random=true
   )]
  particleUpdateScript [
   &ParticleState()
   &ScaleSpriteSize(
    Uniform Curve Scale=&RandomRangeFloat(Minimum=0.1, Maximum=1.5, Evaluation Type=::NewEnumerator1)
    Uniform Curve Sprite Scale=DataInterface() {propertyValues {Curve {keys [<at 0.0387 1>, {tangentMode=RCTM_User, time=0.1142, value=0.3}, {tangentMode=RCTM_User, time=0.8774, value=1}, <at 1.0189 0>]}}}
   )
   &ScaleColor(
    Linear Color Curve=DataInterface() {propertyValues {AlphaCurve {keys [<at 0 1>,<at 1 1>]}, BlueCurve {keys [<at 0 495.0361>]}, GreenCurve {keys [<at 0 0>]}, RedCurve {keys [<at 0 1000>]}}}
    Scale Mode=::NewEnumerator2
   )
   /SetVariables(Particles.Color=&MakeCustomLinearColorFromBool(Bool=&RandomBool(Evaluation Type=::NewEnumerator1), True Color=$Particles.Color, False Color=<rgba 0 0 0 10>)) {enabled=false, isSetParameters=true, module=SetVariables_#2}
   /SetVariables(Particles.SpriteRotation:NiagaraFloat=&FloatFromCurve(
     FloatCurve=DataInterface() {propertyValues {Curve {keys [{arriveTangent=-4.6057, interpMode=RCIM_Cubic, leaveTangent=-4.6057, tangentMode=RCTM_User, time=0, value=2.5}, <at 1 1 RCIM_Cubic>]}}}
     Scale Curve=45
    )) {isSetParameters=true, module=SetVariables_#3}
   /SetVariables(Particles.SpriteSize001:Vector2f=&MakeVector2D(
     X=&MakeFloatFromVector2d(Vector 2D=$Particles.SpriteSize, Channel=::NewEnumerator1)
     Y=&MakeFloatFromVector2d(Vector 2D=$Particles.SpriteSize)
    )) {isSetParameters=true, module=SetVariables_#4}
   &CameraOffset(Camera Offset Amount=5000)
  ]
  rendererClasses [&NiagaraSpriteRendererProperties]
  renderers [
   {rendererClass=&NiagaraSpriteRendererProperties}
   {rendererClass=&NiagaraSpriteRendererProperties, rendererIndex=1}
  ]
 }
 Star001 {
  bEnabled=false
  emitterUpdateScript [
   &EmitterState(Life Cycle Mode=::NewEnumerator1, Loop Behavior=::NewEnumerator1, Loop Delay=2.5, Loop Duration=0.25, UseLoopDelay=true)
   &SpawnRate(SpawnRate=10, Use Spawn Probability=false)
   &SpawnBurst_Instantaneous(Use Spawn Probability=false, UseLoopCountLimit=false) {enabled=false}
  ]
  particleSpawnScript [&InitializeParticle(
    Color Mode=::NewEnumerator0
    Lifetime=0.2
    Sprite Rotation Angle=45
    Sprite Rotation Mode=::NewEnumerator0
    Sprite Size=&Multiply_Vector2D_ByFloat(Float=1000, Vector2D=<20 2>)
    Sprite Size Mode=::NewEnumerator4
    UsePositionOffset=false
    Write Material Random=true
   )]
  particleUpdateScript [
   &ParticleState()
   &ScaleSpriteSize(
    Uniform Curve Scale=&RandomRangeFloat(Minimum=0.1, Maximum=1.5, Evaluation Type=::NewEnumerator1)
    Uniform Curve Sprite Scale=DataInterface() {propertyValues {Curve {keys [<at 0.0387 1>, {tangentMode=RCTM_User, time=0.1142, value=0.3}, {tangentMode=RCTM_User, time=0.8774, value=1}, <at 1.0189 0>]}}}
   )
   &ScaleColor(
    Linear Color Curve=DataInterface() {propertyValues {AlphaCurve {keys [<at 0 1>,<at 1 1>]}, BlueCurve {keys [<at 0 495.0361>]}, GreenCurve {keys [<at 0 0>]}, RedCurve {keys [<at 0 1000>]}}}
    Scale Mode=::NewEnumerator2
   )
   &ShapeLocation(MaskSpawnGroup=false, Override Seed=false, Sphere Radius=5000, Sphere Surface Distribution=1, UseOwnerScale=true)
   /SetVariables(Particles.Color=&MakeCustomLinearColorFromBool(Bool=&RandomBool(Evaluation Type=::NewEnumerator1), True Color=$Particles.Color, False Color=<rgba 0 0 0 10>)) {isSetParameters=true, module=SetVariables_#2}
   /SetVariables(Particles.SpriteRotation:NiagaraFloat=&FloatFromCurve(
     FloatCurve=DataInterface() {propertyValues {Curve {keys [{arriveTangent=-3.3589, interpMode=RCIM_Cubic, leaveTangent=-3.3589, tangentMode=RCTM_User, time=0, value=2}, <at 1 1 RCIM_Cubic>]}}}
     Scale Curve=45
    )) {enabled=false, isSetParameters=true, module=SetVariables_#3}
   /SetVariables(Particles.SpriteSize001:Vector2f=&MakeVector2D(
     X=&MakeFloatFromVector2d(Vector 2D=$Particles.SpriteSize, Channel=::NewEnumerator1)
     Y=&MakeFloatFromVector2d(Vector 2D=$Particles.SpriteSize)
    )) {isSetParameters=true, module=SetVariables_#4}
  ]
  rendererClasses [&NiagaraSpriteRendererProperties]
  renderers [
   {rendererClass=&NiagaraSpriteRendererProperties}
   {rendererClass=&NiagaraSpriteRendererProperties, rendererIndex=1}
  ]
 }
 Star002 {
  emitterUpdateScript [
   &EmitterState()
   &SpawnBurst_Instantaneous(Use Spawn Probability=false, UseLoopCountLimit=false)
  ]
  particleSpawnScript [&InitializeParticle(
    Color Mode=::NewEnumerator0
    Sprite Rotation Mode=::NewEnumerator0
    Sprite Size=&Multiply_Vector2D_ByFloat(Float=1000, Vector2D=<80 50>)
    Sprite Size Mode=::NewEnumerator4
    UsePositionOffset=false
    Write Material Random=true
   )]
  particleUpdateScript [
   &ParticleState()
   &ScaleSpriteSize(
    Uniform Curve Scale=&RandomRangeFloat(Minimum=0.1, Maximum=1.5, Evaluation Type=::NewEnumerator1)
    Uniform Curve Sprite Scale=DataInterface() {propertyValues {Curve {keys [<at 0.0387 1>, {tangentMode=RCTM_User, time=0.1142, value=0.3}, {tangentMode=RCTM_User, time=0.8774, value=1}, <at 1.0189 0>]}}}
   )
   &ScaleColor(
    Linear Color Curve=DataInterface() {propertyValues {AlphaCurve {keys [<at 0 0.8>]}, BlueCurve {keys [<at 0 0.2475>]}, GreenCurve {keys [<at 0 0>]}, RedCurve {keys [<at 0 0.5>]}}}
    Scale Mode=::NewEnumerator2
   )
   /SetVariables(Particles.Color=&MakeCustomLinearColorFromBool(
     Bool=&RandomBool(Evaluation Type=::NewEnumerator1)
     True Color=$Particles.Color
     False Color=<rgba 1 0 0 0.8>
    )) {isSetParameters=true, module=SetVariables_#2}
   /SetVariables(Particles.SpriteRotation:NiagaraFloat=&FloatFromCurve(
     FloatCurve=DataInterface() {propertyValues {Curve {keys [{arriveTangent=-4.6057, interpMode=RCIM_Cubic, leaveTangent=-4.6057, tangentMode=RCTM_User, time=0, value=2.5}, <at 1 1 RCIM_Cubic>]}}}
     Scale Curve=45
    )) {isSetParameters=true, module=SetVariables_#3}
   /SetVariables(Particles.SpriteSize001:Vector2f=&MakeVector2D(
     X=&MakeFloatFromVector2d(Vector 2D=$Particles.SpriteSize, Channel=::NewEnumerator1)
     Y=&MakeFloatFromVector2d(Vector 2D=$Particles.SpriteSize)
    )) {isSetParameters=true, module=SetVariables_#4}
   &CameraOffset(Camera Offset Amount=-5000)
  ]
  rendererClasses [&NiagaraSpriteRendererProperties]
  renderers [
   {rendererClass=&NiagaraSpriteRendererProperties}
   {rendererClass=&NiagaraSpriteRendererProperties, rendererIndex=1}
  ]
 }
 Void {
  emitterUpdateScript [
   &EmitterState()
   &SpawnBurst_Instantaneous(Use Spawn Probability=false, UseLoopCountLimit=false)
  ]
  particleSpawnScript [&InitializeParticle(
    Color=<rgba -0.3539 -4.6333 -2.1168 1>
    Sprite Size Mode=::NewEnumerator3
    Uniform Sprite Size=10000
    UsePositionOffset=false
    Write Material Random=true
   )]
  particleUpdateScript [
   &ParticleState()
   &ScaleSpriteSize(
    Uniform Curve Scale=&RandomRangeFloat(Minimum=1, Maximum=1.01, Evaluation Type=::NewEnumerator1)
    Uniform Curve Sprite Scale=DataInterface() {propertyValues {Curve {keys [{arriveTangent=5.2524, time=0.7, value=1}, {leaveTangent=24.4158, time=1, value=8}]}}}
   )
   &CameraOffset(Camera Offset Amount=&FloatFromCurve(FloatCurve=DataInterface() {propertyValues {Curve {keys [<at 0 -1>,<at 1 1>]}}}, Scale Curve=9999))
   &ScaleColor(
    Linear Color Curve=DataInterface() {propertyValues {
      AlphaCurve {keys [<at 0.207 0.1>,<at 0.5311 1>,<at 0.7438 1>,<at 1 0>]}
      BlueCurve {keys [<at 0.0045 0>,<at 1 1>]}
      GreenCurve {keys [<at 0.0045 0>,<at 1 1>]}
      RedCurve {keys [<at 0.0045 0>,<at 1 1>]}
     }}
    Scale Mode=::NewEnumerator2
   )
  ]
  rendererClasses [&NiagaraSpriteRendererProperties]
  renderers [{rendererClass=&NiagaraSpriteRendererProperties}]
 }
 lines {
  emitterUpdateScript [
   &EmitterState()
   &SpawnRate(SpawnRate=50, Use Spawn Probability=false)
   &SpawnBurst_Instantaneous(Spawn Count=20, Use Spawn Probability=false, UseLoopCountLimit=false)
  ]
  particleSpawnScript [
   &InitializeParticle(
    Color=<rgba 1000 0 667.5887 1>
    Lifetime=&RandomRangeFloat(Minimum=0.1, Maximum=0.25)
    Sprite Size=&Multiply_Vector2D_ByFloat(Float=100, Vector2D=&RandomRangeVector2D(Minimum=<1 50>, Maximum=<3 50>))
    Sprite Size Mode=::NewEnumerator4
    UsePositionOffset=false
    Write Material Random=true
   )
   /SetVariables(Particles.Velocity=&RandomVector(Vector Scale=1)) {isSetParameters=true}
  ]
  particleUpdateScript [
   &ParticleState()
   &ScaleSpriteSize(
    Scale Factor=&MakeVector2D(
     X=1
     Y=&FloatFromCurve(
      FloatCurve=DataInterface() {propertyValues {Curve {keys [{arriveTangent=-3.3575, interpMode=RCIM_Cubic, leaveTangent=-3.3575, tangentMode=RCTM_User, time=0.3, value=1}, <at 1 0.2 RCIM_Cubic>]}}}
      Scale Curve=10
     )
    )
    Scale Sprite Size Mode=::NewEnumerator2
   )
   /SetVariables(Particles.Position=&AddVectorToPosition(
     Position=$Engine.Emitter.SimulationPosition
     Vector=&Multiply_VectorByFloat(
      Vector=$Particles.Velocity
      Float=&FloatFromCurve(
       FloatCurve=DataInterface() {propertyValues {Curve {keys [{arriveTangent=-2.3071, interpMode=RCIM_Cubic, leaveTangent=-2.3071, tangentMode=RCTM_User, time=0, value=1}, <at 0.7 0.2 RCIM_Cubic>]}}}
       Scale Curve=50000
      )
     )
    )) {isSetParameters=true}
   &SolveForcesAndVelocity(Clamp Velocity=false, Limit Acceleration=false) {enabled=false}
  ]
  rendererClasses [&NiagaraSpriteRendererProperties]
  renderers [{rendererClass=&NiagaraSpriteRendererProperties}]
 }
}
summary {
 systemName=NS_Grillmoon_Katana_ChargeFX_V2
 emitters [
  {emitterName=Star, rendererClasses [&NiagaraSpriteRendererProperties]}
  {emitterName=lines, rendererClasses [&NiagaraSpriteRendererProperties]}
  {emitterName=Void, rendererClasses [&NiagaraSpriteRendererProperties]}
  {bEnabled=false, emitterName=Star001, rendererClasses [&NiagaraSpriteRendererProperties]}
  {emitterName=Star002, rendererClasses [&NiagaraSpriteRendererProperties]}
 ]
 userVariables []
}
