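# Niagara system for a katana slash trail, filed under the Demon kaiju fight FX, drawn as one very large, bright red ribbon.
# SpawnPerFrame places each point far along the emitter's local Z from its current position, with custom facing on that axis, so the trail follows emitter movement.
# Ribbon width grows over the emitter loop, alpha fades late in particle life, and emitter age dims RGB across the whole ribbon.
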NiagaraSystem /GrillMoon/Assets/FX/Kaiju/Fight/Demon/Sword/NS_Grillmoon_Katana_Slash
dependencies {
 usedDataInterfaces [&NiagaraDataInterfaceCurve, &NiagaraDataInterfaceColorCurve]
 usedDynamicInputs []
 usedModules [
  &self:SystemSpawnScript.SystemScriptSource.SystemScriptGraph.NiagaraNodeAssignment_1.SetVariables_#1
  &SystemState
 ]
}
emitters {Ribbon {
  emitterUpdateScript [
   &EmitterState()
   &SpawnPerFrame(Use Spawn Probability=false)
   &SpawnRate(SpawnRate=15, Use Spawn Probability=false) {enabled=false}
  ]
  particleSpawnScript [
   &InitializeParticle(
    Color=<rgba 500 0 0 1>
    Position=&AddVectorToPosition(
     Position=$Engine.Emitter.SimulationPosition
     Vector=&TransformVector(Vector=<0 0 10000>, Destination Space=::Simulation)
    )
    Position Mode=::NewEnumerator1
    Ribbon Facing Mode=::NewEnumerator1
    Ribbon Width=&FloatFromCurve(
     FloatCurve=DataInterface() {propertyValues {Curve {keys [{arriveTangent=1.928, interpMode=RCIM_Cubic, leaveTangent=1.928, tangentMode=RCTM_User, time=0, value=0}, <at 1 1 RCIM_Cubic>]}}}
     CurveIndex=$Emitter.NormalizedLoopAge
     Scale Curve=12000
    )
    Ribbon Width Mode=::NewEnumerator1
    UsePositionOffset=false
    Write Material Random=true
   )
   /SetVariables(Particles.RibbonFacing=&TransformVector(Vector=<0 0 1>, Destination Space=::Simulation)) {isSetParameters=true}
  ]
  particleUpdateScript [
   &ParticleState()
   &ScaleColor(
    Linear Color Curve=DataInterface() {propertyValues {
      AlphaCurve {keys [<at 0.5514 1>,<at 0.7357 0>]}
      BlueCurve {keys [<at 0.0247 1>]}
      GreenCurve {keys [<at 0.0247 1>]}
      RedCurve {keys [<at 0.0247 1>]}
     }}
    Scale Mode=::NewEnumerator2
   )
   /SetVariables(Particles.Position=&AddVectorToPosition(
     Position=$Particles.Initial.Position
     Vector=&TransformVector(
      Vector=&Multiply_VectorByFloat(Vector=<0 0 0.5>, Float=$Particles.RibbonWidth)
      Destination Space=::Simulation
     )
    )) {enabled=false, isSetParameters=true}
   &ScaleRibbonWidth(Ribbon Width Scale=&FloatFromCurve(FloatCurve=DataInterface() {propertyValues {Curve {keys [<at 0 1>,<at 1 0>]}}})) {enabled=false}
   &ScaleColor(
    Color Value To Scale=$Particles.Color
    Scale RGB=&VectorFromFloat(Value=&FloatFromCurve(FloatCurve=DataInterface() {propertyValues {Curve {keys [<at 0.3 0.95>,<at 0.7 0>]}}}, CurveIndex=$Emitter.Age))
    ScaleA=true
    ScaleRGB=true
   ) {module=ScaleColor001}
  ]
  rendererClasses [&NiagaraRibbonRendererProperties]
  renderers [{rendererClass=&NiagaraRibbonRendererProperties}]
 }}
summary {
 systemName=NS_Grillmoon_Katana_Slash
 emitters [{emitterName=Ribbon, rendererClasses [&NiagaraRibbonRendererProperties]}]
 userVariables []
}
